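# M_RiverWaterPreview is an unlit translucent lookdev material previewing river water, usable on spline meshes.
# World-space flow-mapped normals drive a cubemap reflection, added to Fresnel-weighted scene color absorbed exponentially by water depth, plus a faint debug sprite.
# An opacity-style depth fade cuts out below Lake Z and scales down by MPC_Water's EnableTerrainMapping.
# Feature level switches substitute flat color and constant half value on ES2 and ES3.1.
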Material /Clyde/Materials/Lookdev/M_RiverWaterPreview
graph {
 properties {
  bUsedWithSplineMeshes=true
  blendMode=BLEND_Translucent
  pixelDepthOffsetMode=PDOM_Legacy
  shadingModel=MSM_Unlit
  translucencyLightingMode=TLM_SurfacePerPixelLighting
  usedShadingModels=Unlit
 }
 nodes [
  DepthFade(FadeDistance=ScalarParameter() {parameterName=adsfasf})
  n1 = Multiply(
   A=ComponentMask(None=WorldPosition().XYZ) {g=true, r=true}
   B=ComponentMask(None=VectorParameter() {parameterName=Tiling, defaultValue=<rgba 0.0005 0.0005 0 0>}.RGB) {g=true, r=true}
  )
  n2 = MaterialFunctionCall(
   Normal=MaterialFunctionCall!(
    Diffuse
    Normal Map=TextureObject() {samplerType=SAMPLERTYPE_Normal, texture=/Engine/Functions/Engine_MaterialFunctions02/ExampleContent/Textures/water_n}
    "Flow Vector Map (see tooltip)"=VectorParameter() {parameterName=Direction}.RGB
    UV texture 2 offset
    "Time (see tooltip)"
    UVs=@n1
   ) {materialFunction=/Engine/Functions/Engine_MaterialFunctions02/Texturing/FlowMaps_Simple}.Normal
   Flatness=ScalarParameter() {defaultValue=0.7, parameterName=Normal Flatness}
  ) {materialFunction=&FlattenNormal}
  n3 = TextureSampleParameterCube!(UVs=ReflectionVectorWS(CustomWorldNormal=Transform(None=@n2.Result)), Bias, Apply View MipBias) {mipValueMode=TMVM_MipBias, parameterName=UnlitCubemap}
  n4 = VectorParameter() {parameterName=Ref, defaultValue=<rgba 1 1 1 1>}
  n5 = Multiply(A=@n4.RGB, B=@n4.RGB)
  n6 = OneMinus(None=Fresnel())
  n7 = SceneDepth()
  n8 = PixelDepth()
  n9 = VectorParameter() {parameterName=Spectrum, defaultValue=<rgba 10 5 1 0.002>}
  MaterialExpressionMultiply_0 = Multiply(A=Max(A=Subtract(A=@n7.R, B=@n8.R)) {constB=0}, B=Multiply(A=@n9.RGB, B=@n9.RGB))
  n10 = Custom!(x=@MaterialExpressionMultiply_0) {code="exp(-x);", inputs [{inputName=x, input {expression=&self:MaterialExpressionMultiply_0}}]}
  n11 = Multiply(A=@n10, B=SceneColor!(UVs).RGB)
  n12 = Add(
   A=Add(A=Multiply(A=@n3.RGB, B=@n5), B=Multiply(A=@n6, B=@n11) {desc=E})
   B=Multiply(
    A=ScalarParameter() {defaultValue=0.015, parameterName=Debug Text}
    B=TextureSample(UVs=@n1) {texture=/Engine/EditorResources/BadBlueprintSprite}.RGB
   )
  )
  n13 = VectorParameter() {parameterName=MobileColor, defaultValue=<rgba 0.05 0.2 0.5 0>}
  FeatureLevelSwitch(Default=@n12, ES2=@n13.RGB, ES3_1=@n13.RGB, SM4_REMOVED=@n12)
  n14 = WorldPosition()
  n15 = Ceil(None=Subtract(A=ComponentMask(None=@n14.XYZ) {b=true}, B=ScalarParameter() {defaultValue=-500000, parameterName=Lake Z}))
  n16 = Multiply(
   A=DepthFade(Opacity=Saturate(None=@n15), FadeDistance=ScalarParameter() {defaultValue=32, parameterName=FD})
   B=OneMinus(None=CollectionParameter() {collection=/Water/Materials/MPC/MPC_Water, parameterName=EnableTerrainMapping})
  )
  n17 = Constant() {r=0.5}
  FeatureLevelSwitch(Default=@n16, ES2=@n17, ES3_1=@n17, SM4_REMOVED=@n16)
  Multiply(B=@n3.RGB)
  ScalarParameter() {defaultValue=0.05, parameterName=Roughness}
  ScalarParameter() {defaultValue=0.25, parameterName=Specular}
  TextureCoordinate() {vTiling=4}
 ]
}
